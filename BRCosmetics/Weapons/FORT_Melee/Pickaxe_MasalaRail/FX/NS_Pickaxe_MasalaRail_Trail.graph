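# Niagara swing trail effect for the MasalaRail pickaxe, built from two animation-driven ribbon emitters, MainTrail and TipTrail.
# Both spawn anim-trail particles from bone or socket indices, fade by a system swing-fade value and set dynamic material parameters.
# Each tints its ribbon through a color curve indexed by Fortnite's time-of-day parameter.
# TipTrail is shorter-lived, tapers its width over life and adds a time-of-day alpha scalar.
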NiagaraSystem /BRCosmetics/Weapons/FORT_Melee/Pickaxe_MasalaRail/FX/NS_Pickaxe_MasalaRail_Trail
dependencies {
 usedDataInterfaces [&NiagaraDataInterfaceColorCurve, &NiagaraDataInterfaceCurve, &NiagaraDataInterfaceSocketReader]
 usedDynamicInputs [
  &Multiply_Vector
  &MakeVectorFromLinearColorRGB
  &MergeHelper_FloatSwitch
 ]
 usedModules [
  &self:SystemSpawnScript.SystemScriptSource.SystemScriptGraph.NiagaraNodeAssignment_1.SetVariables_#1
  &self:SystemSpawnScript.SystemScriptSource.SystemScriptGraph.NiagaraNodeAssignment_5.SetVariables_#2
  &self:SystemSpawnScript.SystemScriptSource.SystemScriptGraph.NiagaraNodeAssignment_2.SetVariables_#3
  &SystemState
  &self:SystemSpawnScript.SystemScriptSource.SystemScriptGraph.NiagaraNodeAssignment_4.SetVariables_#4
 ]
}
emitters {
 MainTrail {
  emitterSpawnScript [&FN_RibbonIDOffsetUtility()]
  emitterUpdateScript [
   &EmitterState()
   &FN_SpawnAnimTrail()
  ]
  particleSpawnScript [
   &FN_InitializeAnimTrailBoneIndecies()
   &FN_InitializeAnimTrail(Lifetime=0.3)
  ]
  particleUpdateScript [
   &FN_UpdateAnimTrail() {module=FN_UpdateAnimTrail001}
   &ScaleColor(Scale Alpha=$System.SwingGeneralFade, ScaleA=true, ScaleRGB=true) {module=ScaleColor001}
   &DynamicMaterialParameters(
    Index 0 Param 1=1
    Index 0 Param 4=1
    Index 1 Param 4=0.4
    Index 2 Param 1=30
    Index 2 Param 2=2
    Index 2 Param 4=0.6
    Module.Index0Float0Write=true
    Module.Index0Float1Write=true
    Module.Index0Float2Write=true
    Module.Index0Float3Write=true
    Module.Index1Float0Write=true
    Module.Index1Float1Write=true
    Module.Index1Float2Write=true
    Module.Index1Float3Write=true
    Module.Index2Float0Write=true
    Module.Index2Float1Write=true
    Module.Index2Float2Write=false
    Module.Index2Float3Write=false
    Write Parameter Index 1=::NewEnumerator1
    Write Parameter Index 2=::NewEnumerator1
   )
   &ScaleColor(
    Scale Mode=::NewEnumerator1
    Scale RGBA=&MakeVector4_FromLinearColor(LinearColor=&ColorFromCurve(
      ColorCurve=DataInterface() {propertyValues {
        AlphaCurve {keys [<at 0 0.7>,<at 0.1667 0.7>,<at 0.25 0.8>,<at 0.375 0.675>,<at 0.5 1>,<at 0.625 0.8>,<at 0.75 0.8>,<at 0.875 0.7>,<at 1 0.7>]}
        BlueCurve {keys [<at 0 0.2307>,<at 0.0833 0.1095>,<at 0.125 0.09>,<at 0.1667 0.1119>,<at 0.25 0.0931>,<at 0.375 0.1095>,<at 0.4167 0.15>,<at 0.5 0.15 RCIM_Cubic>,<at 0.5833 0.15>,<at 0.625 0.14>,<at 0.75 0.0977>,<at 0.875 0.0888>,<at 1 0.2307>]}
        GreenCurve {keys [<at 0 0.0825>,<at 0.0833 0.0391>,<at 0.125 0.0924>,<at 0.1667 0.0666>,<at 0.25 0.063>,<at 0.375 0.0561>,<at 0.4167 0.0787>,<at 0.5 0.0787 RCIM_Cubic>,<at 0.5833 0.0787>,<at 0.625 0.1125>,<at 0.75 0.0833>,<at 0.875 0.0666>,<at 1 0.0825>]}
        RedCurve {keys [<at 0 0.0346>,<at 0.0833 0.0267>,<at 0.125 0.15>,<at 0.1667 0.0887>,<at 0.25 0.0953>,<at 0.375 0.0467>,<at 0.4167 0.0787>,<at 0.5 0.0787 RCIM_Cubic>,<at 0.5833 0.0787>,<at 0.625 0.15>,<at 0.75 0.15>,<at 0.875 0.12>,<at 1 0.0346>]}
       }}
      CurveIndex=Linked() {linkedVariable {name=NPC.FortniteNPC.TimeOfDay_0-1, type=NiagaraFloat}}
     ))
    ScaleRGBA=true
   )
  ]
  rendererClasses [&NiagaraRibbonRendererProperties]
  renderers [{rendererClass=&NiagaraRibbonRendererProperties}]
 }
 TipTrail {
  emitterSpawnScript [&FN_RibbonIDOffsetUtility()]
  emitterUpdateScript [
   &EmitterState()
   &FN_SpawnAnimTrail(AnimTrailWidthMode=::FromFirst)
  ]
  particleSpawnScript [
   &FN_InitializeAnimTrailBoneIndecies(First Bone or Socket Index=1, Second Bone or Socket Index=0)
   &FN_InitializeAnimTrail(AnimTrailWidth=0.2, Lifetime=0.225)
  ]
  particleUpdateScript [
   &FN_UpdateAnimTrail() {module=FN_UpdateAnimTrail001}
   &ScaleColor(Scale Alpha=$System.SwingGeneralFade, ScaleA=true, ScaleRGB=true) {module=ScaleColor001}
   &ScaleRibbonWidth(Ribbon Width Scale=&FloatFromCurve(FloatCurve=DataInterface() {propertyValues {Curve {keys [<at 0 1>, {leaveTangent=-1.7704, time=1, value=0}]}}}))
   &DynamicMaterialParameters(
    Index 0 Param 1=1
    Index 0 Param 4=1
    Index 1 Param 4=1
    Index 2 Param 1=30
    Index 2 Param 2=10
    Index 2 Param 4=0.6
    Module.Index0Float0Write=true
    Module.Index0Float1Write=true
    Module.Index0Float2Write=true
    Module.Index0Float3Write=true
    Module.Index1Float0Write=true
    Module.Index1Float1Write=true
    Module.Index1Float2Write=true
    Module.Index1Float3Write=true
    Module.Index2Float0Write=true
    Module.Index2Float1Write=true
    Module.Index2Float2Write=false
    Module.Index2Float3Write=false
    Write Parameter Index 1=::NewEnumerator1
    Write Parameter Index 2=::NewEnumerator1
   )
   &ScaleColor(
    Scale Mode=::NewEnumerator1
    Scale RGBA=&Multiply_Vector4(
     A=&MakeVector4_FromLinearColor(LinearColor=&ColorFromCurve(
       ColorCurve=DataInterface() {propertyValues {
         AlphaCurve {keys [<at 0 0.7>,<at 0.1667 0.7>,<at 0.25 0.8>,<at 0.375 0.675>,<at 0.5 1>,<at 0.625 0.8>,<at 0.75 0.8>,<at 0.875 0.7>,<at 1 0.7>]}
         BlueCurve {keys [<at 0 0.2307>,<at 0.0833 0.1095>,<at 0.125 0.09>,<at 0.1667 0.1119>,<at 0.25 0.0931>,<at 0.375 0.1095>,<at 0.4167 0.15>,<at 0.5 0.15 RCIM_Cubic>,<at 0.5833 0.15>,<at 0.625 0.14>,<at 0.75 0.0977>,<at 0.875 0.0888>,<at 1 0.2307>]}
         GreenCurve {keys [<at 0 0.0825>,<at 0.0833 0.0391>,<at 0.125 0.0924>,<at 0.1667 0.0666>,<at 0.25 0.063>,<at 0.375 0.0561>,<at 0.4167 0.0787>,<at 0.5 0.0787 RCIM_Cubic>,<at 0.5833 0.0787>,<at 0.625 0.1125>,<at 0.75 0.0833>,<at 0.875 0.0666>,<at 1 0.0825>]}
         RedCurve {keys [<at 0 0.0346>,<at 0.0833 0.0267>,<at 0.125 0.15>,<at 0.1667 0.0887>,<at 0.25 0.0953>,<at 0.375 0.0467>,<at 0.4167 0.0787>,<at 0.5 0.0787 RCIM_Cubic>,<at 0.5833 0.0787>,<at 0.625 0.15>,<at 0.75 0.15>,<at 0.875 0.12>,<at 1 0.0346>]}
        }}
       CurveIndex=Linked() {linkedVariable {name=NPC.FortniteNPC.TimeOfDay_0-1, type=NiagaraFloat}}
      ))
     B=&MakeVector4FromVectorAndFloat(
      "Vector (RGB)"=&VectorFromFloat(Value=0.35)
      "Float (Alpha)"=&FNTimeOfDayScalar(Night Scale=1.2, Morning Scale=1.4, Day Scale=2, Evening Scale=1.35)
     )
    )
    ScaleRGBA=true
   )
  ]
  rendererClasses [&NiagaraRibbonRendererProperties]
  renderers [{rendererClass=&NiagaraRibbonRendererProperties}]
 }
}
summary {
 systemName=NS_Pickaxe_MasalaRail_Trail
 emitters [
  {emitterName=MainTrail, rendererClasses [&NiagaraRibbonRendererProperties]}
  {emitterName=TipTrail, rendererClasses [&NiagaraRibbonRendererProperties]}
 ]
 userVariables [
  User.bIsFrontEnd {type=NiagaraBool}
  User.bIsFrontEndPreview {type=NiagaraBool}
  User.RibbonIDOffset {type=NiagaraInt32}
  User.SwingSectionIndex {type=NiagaraInt32}
 ]
}
